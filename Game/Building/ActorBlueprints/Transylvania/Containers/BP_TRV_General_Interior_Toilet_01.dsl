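; Interior toilet container Blueprint for the Transylvania building set that sometimes has a buzzing fly effect.
; On BeginPlay, only instances without authority roll a 20 percent chance to play the flies sound and spawn a Niagara swarm above the root, storing its component.
; Taking damage deactivates the swarm and stops audio.
; Destruction does the same and also destroys the Niagara component.
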(event EventDestroyed
  (bind _ns_flies (Variables|Components|GetNS_Flies))
  ;[1 1 1] AUDIO
  (Components|Activation|Deactivate _ns_flies)
  (Components|DestroyComponent _ns_flies)
  (Audio|Components|Audio|Stop))

(event EventOnDamagePlayEffects
  (params (Damage :type Float%20%28single-precision%29) (DamageTags :type Gameplay%20Tag%20Container%20Structure%20%28by%20ref%29) (Momentum :type Vector :default "0, 0, 0") (HitInfo :type Hit%20Result%20Structure%20%28by%20ref%29) (InstigatedBy :type Fort%20Pawn%20Object%20Reference) (DamageCauser :type Actor%20Object%20Reference) (EffectContext :type Gameplay%20Effect%20Context%20Handle%20Structure))
  ;[1 1 1] AUDIO
  (Components|Activation|Deactivate (Variables|Components|GetNS_Flies))
  (Audio|Components|Audio|Stop))

(event EventBeginPlay
  (Networking|SwitchHasAuthority
    (:Remote
      (Utilities|FlowControl|Sequence
        (:then_0
          (if (< (Math|Random|RandomFloatinRange 0.0 100) 20)
            (Audio|Components|Audio|Play (Variables|Default|GetSound_Flies) 0.0)
            (bind _returnvalue (Niagara|SpawnSystematLocation "/Game/Packages/Fortress_Transylvania/Haunted_Forest/Swamps/Particles/NS_Flies_01.NS_Flies_01" (+ (Transformation|GetWorldLocation (Variables|Transformation|GetRootComponent)) "0,0.000000,115.000000")))
            (Variables|Components|SetNS_Flies _returnvalue)))))))

(fn ConstructionScript)
